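# Sound cue for Inner Tower boss breath vocalizations, named as an add-on to the boss music.
# A Random node picks one of five equally weighted breath wave variations on each play.
# It plays through the critical enemy VO sound class and uses the shared close-range NPC VO attenuation.
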SoundCue /NPCLibrary/NPCs/Inner/Inner/Audio/VO/Boss_Music_Addons/InnerTowerBoss_2021_BossMusic_Breaths_01_Cue
root=Random_0
nodes [
 {id=Random_0, children [WavePlayer_0,WavePlayer_1,WavePlayer_2,WavePlayer_3,WavePlayer_4], properties {weights [1,1,1,1,1]}}
 {
  id=WavePlayer_0
  properties {soundWaveAssetPtr=&/InnerTowerBoss_2021_BossMusic_Breaths_01_010}
 }
 {
  id=WavePlayer_1
  properties {soundWaveAssetPtr=&/InnerTowerBoss_2021_BossMusic_Breaths_01_011}
 }
 {
  id=WavePlayer_2
  properties {soundWaveAssetPtr=&/InnerTowerBoss_2021_BossMusic_Breaths_01_012}
 }
 {
  id=WavePlayer_3
  properties {soundWaveAssetPtr=&/InnerTowerBoss_2021_BossMusic_Breaths_01_013}
 }
 {
  id=WavePlayer_4
  properties {soundWaveAssetPtr=&/InnerTowerBoss_2021_BossMusic_Breaths_01_014}
 }
]
properties {
 attenuationSettings=/Game/Sounds/Fort_Human_VO_BR/NPCs/Shared/NPC_VO_Close_Att
 maxDistance=8500
 soundClassObject=&Fort_VO_Enemy_Critical
 volumeMultiplier=1
}
